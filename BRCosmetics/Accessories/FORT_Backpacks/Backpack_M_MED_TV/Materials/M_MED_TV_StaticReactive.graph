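# Character material for the MED TV backpack that shows TV static and rolling distortion on a masked screen area.
# Three panned static samples form emissive noise scaled by BP_GlitchesIntensity, which also darkens base color there.
# A time-driven sweep band and the noise offset Diffuse UVs, confined by a SkinFX_Mask channel.
# PreFX and PostFX character functions bracket the effect, with a glitch-subfunction mask feeding clear coat.
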Material /BRCosmetics/Accessories/FORT_Backpacks/Backpack_M_MED_TV/Materials/M_MED_TV_StaticReactive
graph {
 properties {
  bUseMaterialAttributes=true
  bUsedWithMeshDeformer=true
  bUsedWithSkeletalMesh=true
  pixelDepthOffsetMode=PDOM_Legacy
  shadingModel=MSM_FromMaterialExpression
  subsurfaceProfile=&SS_HeroSkin_02
  usedShadingModels=Default Lit | Clear Coat | Subsurface Profile
 }
 nodes [
  n1 = MaterialFunctionCall() {materialFunction=&BaseFN_CharacterFunction_PreFX}
  n2 = MaterialFunctionCall!(In=TextureObjectParameter() {group=Skin FX, parameterName=SkinFX_Mask, texture=&T_WhitesRGB}) {materialFunction=/Game/Weapons/WeaponSkins/MaterialFunction/MF_GlitchSubfunction}
  n3 = StaticSwitchParameter(True=@n2.Glitched, False=@n2.Glitched) {group=Skin FX, parameterName=UseGlitch}
  n4 = GetMaterialAttributes(None=MaterialFunctionCall!(Alpha=StaticComponentMaskParameter(None=@n3) {defaultR=true, group=Skin FX, parameterName=Clear Coat Channel}, In=@n1.Result) {materialFunction=/Game/Weapons/WeaponSkins/MaterialFunction/Clearcoat}.Result) {attributeGetTypes [#1, #2]}
  n5 = Multiply(A=TextureSample(UVs=Panner() {speedX=5, speedY=0.1}) {samplerType=SAMPLERTYPE_Masks, texture=/Game/Missions/Primary/RepairTheShelter/Textures/Cinematic_Screen_Static}.B) {constB=0.1}
  n6 = Multiply(
   A=TextureSample(UVs=Panner() {speedX=6, speedY=-0.05}) {samplerType=SAMPLERTYPE_Masks, texture=&Cinematic_Screen_Static}.G
   B=TextureSample(UVs=Panner() {speedX=-3, speedY=-0.03}) {samplerType=SAMPLERTYPE_Masks, texture=&Cinematic_Screen_Static}.R
  )
  n7 = ScalarParameter() {group=TV Glitches, parameterName=BP_GlitchesIntensity}
  n8 = Multiply(A=Multiply(A=Add(A=@n5, B=@n6), B=ScalarParameter() {defaultValue=3, group=TV Glitches, parameterName=Static Scale}), B=@n7)
  n9 = StaticComponentMaskParameter(None=TextureSample(Tex=TextureObjectParameter() {group=Skin FX, parameterName=SkinFX_Mask, texture=&T_WhitesRGB}).RGB) {defaultR=true, group=Skin FX, parameterName=Static Noise Channel}
  n10 = TextureCoordinate()
  n11 = TextureCoordinate()
  n12 = Multiply(A=OneMinus(None=ComponentMask(None=@n11) {g=true})) {constB=0.75}
  n13 = Multiply(A=Time(), B=ScalarParameter() {defaultValue=-0.2, group=TV Glitches, parameterName=Sweep Speed}) {constB=-0.2}
  n14 = Power(Base=Frac(None=Add(A=@n12, B=@n13))) {constExponent=20}
  n15 = ScalarParameter() {defaultValue=0.3, group=TV Glitches, parameterName=Sweep Add}
  n16 = LinearInterpolate(A=ScalarParameter() {defaultValue=0.3, group=TV Glitches, parameterName=Sweep Add Base}, B=@n15, Alpha=@n7)
  n17 = Multiply(A=@n14, B=ScalarParameter() {defaultValue=0.01, group=TV Glitches, parameterName=Sweep UV Distort}) {constB=0.02}
  n18 = Constant()
  n19 = LinearInterpolate(
   A=ScalarParameter() {defaultValue=0.07, group=TV Glitches, parameterName=Static UV Distort Base}
   B=ScalarParameter() {defaultValue=0.07, group=TV Glitches, parameterName=Static UV Distort}
   Alpha=@n7
  )
  n20 = Add(A=Multiply(A=Multiply(A=@n14, B=@n16) {constB=0.2}, B=AppendVector(A=@n17, B=@n18)), B=AppendVector(A=Multiply(A=@n6, B=@n19), B=Constant()))
  n21 = TextureSampleParameter2D(UVs=LinearInterpolate(A=@n10, B=Add(A=@n10, B=@n20), Alpha=@n9)) {group=Diffuse, parameterName=Diffuse, texture=/Engine/EngineMaterials/DefaultDiffuse}
  n22 = Multiply(A=@n21.RGB, B=OneMinus(None=@n7))
  MaterialFunctionCall(
   AO=@n1.AO
   Distance Blend=@n1.AO
   Fresnel=@n1.AO
   Material=SetMaterialAttributes!(
    MaterialAttributes=@n4.MaterialAttributes
    Emissive Color=LinearInterpolate(A=@n4.Emissive Color, B=@n8, Alpha=@n9)
    Base Color=LinearInterpolate(A=@n21.RGB, B=@n22, Alpha=@n9)
   ) {attributeSetTypes [#1, #2]}
  ) {materialFunction=&BaseFN_CharacterFunction_PostFX}
  Multiply(A=@n7, B=@n15)
  Multiply(A=@n7, B=@n19)
 ]
}
