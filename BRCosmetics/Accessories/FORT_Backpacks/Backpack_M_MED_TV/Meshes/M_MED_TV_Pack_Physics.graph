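# Physics asset for the medium male TV backpack cosmetic, providing physics bodies for its left and right antennas.
# Each antenna is a three-capsule chain, with constraints linking the master bone to a short base segment and that to a longer upper segment.
# The upper joints lock twist and one swing axis, so each tip bends like a hinge, with differing swing limit values on left and right.
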PhysicsAsset /BRCosmetics/Accessories/FORT_Backpacks/Backpack_M_MED_TV/Meshes/M_MED_TV_Pack_Physics
bodies {
 dyn_antenna_l1 [{shapeType=Capsule, length=1, radius=0.3, center=<0.1 0 0>, rotation=<rot 90 0 0>}]
 dyn_antenna_l3 [{shapeType=Capsule, length=3.1185, radius=0.1886, center=<1.7557 0 0>, rotation=<rot 90 0 0>}]
 dyn_antenna_r1 [{shapeType=Capsule, length=1, radius=0.3, center=<0.1 0 0>, rotation=<rot 90 0 0>}]
 dyn_antenna_r3 [{shapeType=Capsule, length=3.257, radius=0.1755, center=<1.8225 0 0>, rotation=<rot 90 0 0>}]
 dyn_master_antenna_l [{shapeType=Capsule, length=0.505, radius=0.505, rotation=<rot 0 -90 0>}]
 dyn_master_antenna_r [{shapeType=Capsule, length=0.505, radius=0.505, rotation=<rot 0 -90 0>}]
}
constraints [
 {bone1Name=dyn_antenna_r1, bone2Name=dyn_master_antenna_r, swing1LimitDegrees=10, swing2LimitDegrees=10, twistLimitDegrees=1}
 {bone1Name=dyn_antenna_l1, bone2Name=dyn_master_antenna_l, swing1LimitDegrees=10, swing2LimitDegrees=10, twistLimitDegrees=1}
 {
  bone1Name=dyn_antenna_r3
  bone2Name=dyn_antenna_r1
  swing1LimitDegrees=20
  swing2LimitDegrees=45
  swing2Motion=Locked
  twistLimitDegrees=45
  twistMotion=Locked
 }
 {
  bone1Name=dyn_antenna_l3
  bone2Name=dyn_antenna_l1
  swing1LimitDegrees=10
  swing2LimitDegrees=45
  swing2Motion=Locked
  twistLimitDegrees=45
  twistMotion=Locked
 }
]
